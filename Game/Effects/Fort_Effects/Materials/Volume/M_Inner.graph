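# Default Lit material for a small emissive screen that shows a tinted icon (a specialty fragment by default) behind a pixel grid, usable on skeletal meshes.
# The icon's red and alpha channels shape image and background brightness, colored by an emissive tint.
# Scrolling analog scanlines on a separate UV channel blend in by intensity.
# A sine over time adds slight flicker, varied by per-instance random, while the pixel grid also scales roughness.
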Material /Game/Effects/Fort_Effects/Materials/Volume/M_Inner
graph {
 properties {bUsedWithMeshDeformer=true, bUsedWithSkeletalMesh=true, pixelDepthOffsetMode=PDOM_Legacy, usedShadingModels=Default Lit}
 nodes [
  n1 = Time()
  n2 = Sine(None=Multiply(A=@n1, B=ScalarParameter() {defaultValue=1, parameterName=Screen_EmissiveFlickerSpeed}))
  n3 = LinearInterpolate(
   A=Clamp(None=ConstantBiasScale(None=@n2) {bias=1.5, scale=0.99})
   B=Clamp(None=ConstantBiasScale(None=@n2) {bias=1.5, scale=0.98})
   Alpha=Frac(None=PerInstanceRandom())
  )
  n4 = VectorParameter() {parameterName=ScreenEmissiveColor, defaultValue=<rgba 0.0598 0.265 0.1435 0>}
  n5 = TextureSampleParameter2D(UVs=TextureCoordinate() {coordinateIndex=1}) {parameterName=Icon_Texture, texture=/Game/Effects/Fort_Effects/Textures/Icons/T_DimFrag_Specialty}
  n6 = Add(
   A=@n5.R
   B=LinearInterpolate(A=@n5.R, B=OneMinus(None=@n5.A), Alpha=ScalarParameter() {defaultValue=0.125, parameterName=ScreenBGD_Darkness}) {constAlpha=0.12}
  )
  n7 = Multiply(A=@n4.RGB, B=@n6)
  n8 = TextureSample(UVs=Multiply(A=TextureCoordinate() {coordinateIndex=1}, B=ScalarParameter() {defaultValue=8, parameterName=Screen_PixelGrid_Tiling})) {texture=/Game/Effects/Fort_Effects/Textures/Tile/T_FragmentScreen_PixelGrid_Tile}
  n9 = Add(
   A=Multiply(A=@n1, B=ScalarParameter() {defaultValue=0.125, parameterName=Screen_ScanlineScrollSpeed}) {constB=0.12}
   B=TextureCoordinate() {coordinateIndex=2}
  )
  n10 = Multiply(A=TextureCoordinate() {coordinateIndex=2}, B=ScalarParameter() {defaultValue=8, parameterName=Screen_PixelGrid_Tiling})
  n11 = Multiply(
   A=ScalarParameter() {defaultValue=1, parameterName=ScanlineBrightness}
   B=Multiply(
    A=TextureSample(UVs=@n9) {texture=/Game/Effects/Fort_Effects/Textures/Tile/T_ScanLine_Analog}.RGB
    B=TextureSample(UVs=@n10) {texture=&T_FragmentScreen_PixelGrid_Tile}.RGB
   )
  )
  n12 = Multiply(A=Multiply(A=@n4.RGB, B=Add(A=@n4.RGB, B=@n11)), B=@n6)
  LinearInterpolate(
   A=Multiply(A=@n3, B=Multiply(A=@n7, B=@n8.RGB))
   B=Multiply(A=@n3, B=Multiply(A=@n12, B=@n8.RGB))
   Alpha=ScalarParameter() {defaultValue=0.5, parameterName=ScreenScanlineIntensity}
  )
  Multiply(A=ScalarParameter() {defaultValue=1, parameterName=Screen_Roughness}, B=@n8.RGB)
  VectorParameter() {parameterName=ScreenBaseColor, defaultValue=<rgba 0.0011 0.025 0.0101 0>}
 ]
}
